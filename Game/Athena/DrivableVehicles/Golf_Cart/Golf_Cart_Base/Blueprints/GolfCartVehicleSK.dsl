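; Golf cart drivable vehicle Blueprint handling cosmetic feedback and player interactions.
; Enter/exit events track driver and passengers, binding camera shakes, headlights, impulse-weapon components and an aim-down-sights material swap that prevents occlusion.
; Movement, boost, water, collision, health and death events drive audio, haptics, particles and material parameters, with audio gated by listener range and significance LOD.
; Hit pawns are launched with instigator credit; roof overlaps bounce players upward after a ceiling trace.
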(event Vehicle|Water|EventOnEnteredWaterVolume
  (params (WaterSurfacePoint :type Vector%20%28by%20ref%29 :default "0, 0, 0"))
  (bind _returnvalue (CallFunction|GetWaterLevel))
  (bind (_returnvalue_x _returnvalue_y _returnvalue_z) (Math|Vector|BreakVector (Transformation|GetActorLocation)))
  (bind _const_waterfxoffsetz (Variables|Default|GetConst_WaterFXOffsetZ))
  (bind _returnvalue_2 (+ _returnvalue _const_waterfxoffsetz))
  (bind _returnvalue_3 (- _returnvalue_2 _returnvalue_z))
  (bind _skeletalmesh (Variables|Vehicle|GetSkeletalMesh))
  ;[1 1 1] FX for water ineraction
  (Networking|CanExecuteCosmeticEvents
    (:True
      (if (> (Rendering|GetRenderingDetailMode) 0)
        (bind _returnvalue_4 (Utilities|Time|SetTimerbyEvent (delegate Custom|CheckWater) 0.25 true))
        (bind _inwaterfx (Variables|Default|GetInWaterFX)
          (:then
            (Components|DestroyComponent _inwaterfx))
          (:else
            _))
        (bind _returnvalue_5 (Effects|SpawnEmitterAttached :EmitterTemplate "/Game/WaterAndWind/Particles/WaterInteraction/P_Vehicle_In_Water_Athena.P_Vehicle_In_Water_Athena" :AttachToComponent _skeletalmesh :AttachPointName "None" :Location (Math|Vector|MakeVector 0.0 0.0 _returnvalue_3) :Scale "1.000000,1.000000,1.000000" :LocationType "KeepRelativeOffset" :bAutoDestroy true :PoolingMethod "None" :bAutoActivate true)) ; Looping water FX
        (Variables|Default|SetInWaterFX _returnvalue_5)
        (Effects|SpawnEmitterAttached :EmitterTemplate "/Game/WaterAndWind/Particles/WaterInteraction/P_WaterSplash_Athena_Vehicle.P_WaterSplash_Athena_Vehicle" :AttachToComponent _skeletalmesh :AttachPointName "None" :Location (Math|Vector|MakeVector 0.0 0.0 _returnvalue_3) :Scale "1.000000,1.000000,1.000000" :LocationType "KeepRelativeOffset" :bAutoDestroy true :PoolingMethod "None" :bAutoActivate true) ; Splash burst FX
        (bind _inwaterloop (Variables|Default|GetInWaterLoop)
          (:then
            (Audio|Components|Audio|Stop _inwaterloop))
          (:else
            _))
        (Audio|SpawnSoundAttached "/Game/Athena/Sounds/Vehicles/GolfCart/GolfCart_Movement_Impact_Water_Cue.GolfCart_Movement_Impact_Water_Cue" (Variables|Vehicle|GetSkeletalMesh))
        (bind _returnvalue_6 (Audio|SpawnSoundAttached "/Game/Athena/Sounds/Vehicles/GolfCart/GolfCart_Movement_InWater_Loop_Cue.GolfCart_Movement_InWater_Loop_Cue" (Variables|Vehicle|GetSkeletalMesh)))
        (Variables|Default|SetInWaterLoop _returnvalue_6)))))

(event Custom|CheckWater
  (bind (_returnvalue_x _returnvalue_y _returnvalue_z) (Math|Vector|BreakVector (Transformation|GetActorLocation)))
  ;[1 1 1] FX for water ineraction
  (bind _inwaterfx (Variables|Default|GetInWaterFX)
    (:then
      (Transformation|SetRelativeLocation _inwaterfx (Math|Vector|MakeVector 0.0 0.0 (- (+ (CallFunction|GetWaterLevel) (Variables|Default|GetConst_WaterFXOffsetZ)) _returnvalue_z)) false false)
      (Effects|Components|ParticleSystem|SetVectorParameter (Variables|Default|GetInWaterFX) "Offset" (Math|Vector|Normalize (Math|Transform|InverseTransformDirection (Transformation|GetWorldTransform (Variables|Default|GetInWaterFX)) (Transformation|GetVelocity))))
      (bind _inwaterloop (Variables|Default|GetInWaterLoop)
        (:then
          (Audio|Components|Audio|SetVolumeMultiplier _inwaterloop (Math|Float|MapRangeClamped (Vehicle|GetForwardSpeedKmh) 0.0 2e+01 0.0 1.0) :!class /Script/Engine.AudioComponent))
        (:else)))
    (:else)))

(event Vehicle|Water|EventOnExitedWaterVolume
  (params (WaterSurfacePoint :type Vector%20%28by%20ref%29 :default "0, 0, 0"))
  ;[1 1 1] FX for water ineraction
  (Networking|CanExecuteCosmeticEvents
    (:True
      (Utilities|Time|ClearandInvalidateTimerbyHandle (stale Vehicle|Water|EventOnEnteredWaterVolume _returnvalue_4))
      (bind _inwaterfx (Variables|Default|GetInWaterFX)
        (:then
          (Components|Activation|SetActive _inwaterfx false false)
          (bind _inwaterloop (Variables|Default|GetInWaterLoop)
            (:then
              (Audio|Components|Audio|FadeOut _inwaterloop 3.0 0.0 "Linear" :!class /Script/Engine.AudioComponent))
            (:else)))
        (:else)))))

(event Custom|PedalForceFeedback
  (bind _drivingplayercontroller (Variables|Default|GetDrivingPlayerController))
  ;[1 1 1] Force Feedback for "Pedaling" (Called from AnimBP)
  (Utilities|IsValid _drivingplayercontroller
    (:Is%20Valid
      (Game|Feedback|ClientPlayForceFeedback _drivingplayercontroller "/Game/Weapons/ForceFeedback/FF_ShoppingCartPedal.FF_ShoppingCartPedal" "None" false false false)
      ;[1 1 1] Camera shake on push
      (CameraShakes|StartCameraShake :self (Variables|/Script/Engine.PlayerController|GetPlayerCameraManager (Variables|Default|GetDrivingPlayerController)) :ShakeClass "/Game/Athena/DrivableVehicles/CameraShake_Pedal.CameraShake_Pedal_C" :Scale 1.0 :PlaySpace "CameraLocal"))))

(event EventOnPawnExitVehicle
  (params (PlayerPawn :type Fort%20Player%20Pawn%20Object%20Reference) (ExitSocketName :type Name :default "None"))
  (bind _impulsecomponent_passenger3 (Variables|Default|GetImpulseComponent_Passenger3))
  (bind _passengerplayercontroller (Variables|Default|GetPassengerPlayerController))
  (bind _impulsecomponent_passenger1 (Variables|Default|GetImpulseComponent_Passenger1))
  (bind _impulsecomponent_passenger2 (Variables|Default|GetImpulseComponent_Passenger2))
  ;[0 0.1875 0.1503] EVENT -> Pawn Exit Vehicle
  (Ability|GameplayEffect|RemoveLooseGameplayTags PlayerPawn (Variables|Default|GetGameplay%20Tags) true)
  (Utilities|Array|RemoveItem (Variables|Default|GetPassengers) PlayerPawn)
  ;; IsLocalLeaving
  (Utilities|FlowControl|Sequence
    (:then_0
      (if (Utilities|Equal(Object) (Variables|Vehicle|GetLocalPlayerPawn) PlayerPawn)
        (Variables|Vehicle|SetLocalPlayerPawn 0)))
    (:then_1
      _))
  (if (Utilities|Equal(Object) (Variables|Default|GetDriverPawn) PlayerPawn)
    ;; ManageStateDriverExit
    (Utilities|FlowControl|Sequence
      (:then_0
        ;;; ManageHeadlightsOff
        (Utilities|FlowControl|Sequence
          (:then_0
            (TimeOfDay|UnregisterforDayPhaseChangedEvent (TimeOfDay|GetContextualTimeOfDayManager) self)
            (CallFunction|SendHeadlightValuetoMaterials 0.0))
          (:then_1
            _))
        (Variables|Default|SetIgnoreHitLastDriver true)
        (Utilities|Time|SetTimerbyEvent (EventDispatchers|CreateEvent :!event CanHitLastDriver) 3.0)
        (Utilities|FlowControl|Sequence
          (:then_0
            (CallFunction|StopDriverRumble)
            (CallFunction|StopDriverCameraShake)
            (CallFunction|BoostRumbleEnd))
          (:then_1
            (Variables|Default|SetDrivingPlayerController 0))))
      (:then_1
        (Variables|Default|SetDriverPawn 0)
        (Variables|Default|SetLastDriver PlayerPawn))
      (:then_2
        (bind _asb_ranged_generic (Utilities|Casting|CastToB_Ranged_Generic (Variables|/Script/FortniteGame.FortPawn|GetCurrentWeapon (Variables|Default|GetLastDriver))))
        (Default|UnbindEventfromOnAimDownSightsChanged _asb_ranged_generic (EventDispatchers|CreateEvent :!event Custom|LocalPassengerADS)))
      (:then_3
        (FortPawn|UnbindEventfromOnWeaponEquippedDelegate (Variables|Default|GetLastDriver) (EventDispatchers|CreateEvent :!event Custom|OnWeaponEquipped))))
    (else
      ;; ManageStatePassengerExit
      (Utilities|FlowControl|Sequence
        (:then_0
          (Utilities|IsValid (Variables|CameraShake|GetPassengerCameraShake)
            (:Is%20Valid
              (Utilities|IsValid _passengerplayercontroller
                (:Is%20Valid
                  (CameraShakes|StopCameraShake (Variables|/Script/Engine.PlayerController|GetPlayerCameraManager _passengerplayercontroller) (Variables|CameraShake|GetPassengerCameraShake) false)
                  (Variables|CameraShake|SetPassengerCameraShake 0))))))
        (:then_1
          (Utilities|FlowControl|Sequence
            (:then_0
              (if (Utilities|Equal(Object) (Variables|Default|GetPassengerPawn) PlayerPawn)
                (FortPawn|UnbindEventfromOnWeaponEquippedDelegate (Variables|Default|GetPassengerPawn) (EventDispatchers|CreateEvent :!event Custom|OnWeaponEquipped))
                (Development|PrintString "Event unbound Driver" true true "(R=0.743277,G=0.000000,B=1.000000,A=1.000000)" 4.0)))
            (:then_1
              (if (Utilities|Equal(Object) PlayerPawn (Variables|/Game/Athena/DrivableVehicles/PassengerImpulseWeaponComponent.PassengerImpulseWeaponComponent_C|GetPassengerPawn _impulsecomponent_passenger1))
                (Class|PassengerImpulseWeaponComponent|SetPassengerPawn _impulsecomponent_passenger1 0 :!class /Game/Athena/DrivableVehicles/PassengerImpulseWeaponComponent.PassengerImpulseWeaponComponent_C)
                (elif (Utilities|Equal(Object) PlayerPawn (Variables|/Game/Athena/DrivableVehicles/PassengerImpulseWeaponComponent.PassengerImpulseWeaponComponent_C|GetPassengerPawn _impulsecomponent_passenger2))
                  (Class|PassengerImpulseWeaponComponent|SetPassengerPawn _impulsecomponent_passenger2 0 :!class /Game/Athena/DrivableVehicles/PassengerImpulseWeaponComponent.PassengerImpulseWeaponComponent_C)
                  (elif (Utilities|Equal(Object) PlayerPawn (Variables|/Game/Athena/DrivableVehicles/PassengerImpulseWeaponComponent.PassengerImpulseWeaponComponent_C|GetPassengerPawn _impulsecomponent_passenger3))
                    (Class|PassengerImpulseWeaponComponent|SetPassengerPawn _impulsecomponent_passenger3 0 :!class /Game/Athena/DrivableVehicles/PassengerImpulseWeaponComponent.PassengerImpulseWeaponComponent_C)))))))
        (:then_2
          (bind _asb_ranged_generic_2 (Utilities|Casting|CastToB_Ranged_Generic (Variables|/Script/FortniteGame.FortPawn|GetCurrentWeapon PlayerPawn)))
          (Default|UnbindEventfromOnAimDownSightsChanged _asb_ranged_generic_2 (EventDispatchers|CreateEvent :!event Custom|LocalPassengerADS)))
        (:then_3
          (Variables|Default|SetPassengerPlayerController 0)
          (Variables|Default|SetPassengerPawn 0))))))

(event Custom|OnWeaponEquipped
  (params (NewWeapon :type Fort%20Weapon%20Object%20Reference) (PrevWeapon :type Fort%20Weapon%20Object%20Reference))
  ;[1 1 1] ADS binding
  (bind _asfort_player_pawn_athena (Utilities|Casting|CastToFortPlayerPawnAthena (Actor|GetOwner NewWeapon)))
  (if (Pawn|IsLocallyControlled _asfort_player_pawn_athena)
    (Utilities|FlowControl|Sequence
      (:then_0
        (bind _asb_ranged_generic (Utilities|Casting|CastToB_Ranged_Generic PrevWeapon))
        (Default|UnbindEventfromOnAimDownSightsChanged _asb_ranged_generic (EventDispatchers|CreateEvent :!event Custom|LocalPassengerADS)))
      (:then_1
        ;;[1 1 1] Bind to ADS event for current weapon
        (bind _asb_ranged_generic_2 (Utilities|Casting|CastToB_Ranged_Generic NewWeapon))
        (Default|BindEventtoOnAimDownSightsChanged _asb_ranged_generic_2 (EventDispatchers|CreateEvent :!event Custom|LocalPassengerADS))))))

(event Custom|NoShake
  (CallFunction|StopDriverCameraShake)
  (Variables|Default|SetDebug_NoCameraShake true))

(event EventTick
  (params (DeltaSeconds :type Float%20%28single-precision%29))
  ;[1 1 1] EVENT TICK -> update Audio and Haptics
  (Utilities|FlowControl|Sequence
    (:then_0
      (CallFunction|UpdateMovementCosmeticParameters))
    (:then_1
      ;; HapticsOnTick
      (if (CallFunction|IsLocalPlayerRiding)
        (if (Vehicle|ClientIsDriver)
          (Utilities|IsValid (Variables|Default|GetDrivingPlayerController)
            (:Is%20Valid
              (Game|Feedback|PlayDynamicForceFeedback (Variables|Default|GetDrivingPlayerController) (Variables|Rumble|GetRumbleIntensity) -1 false true false true))))))))

(event EventOnPawnEnterVehicle
  (params (PlayerPawn :type Fort%20Player%20Pawn%20Object%20Reference) (SeatIdx :type Integer))
  ;[0 0.1875 0.1503] On Player enter, check for passenger.
  (Ability|GameplayEffect|AddLooseGameplayTags PlayerPawn (Variables|Default|GetGameplay%20Tags) true)
  (Utilities|Array|AddUnique (Variables|Default|GetPassengers) PlayerPawn)
  ;; SetLocalPlayerPawn
  (Utilities|FlowControl|Sequence
    (:then_0
      (if (Pawn|IsLocallyControlled PlayerPawn)
        (Variables|Vehicle|SetLocalPlayerPawn PlayerPawn)))
    (:then_1
      _))
  (switch int SeatIdx
    (:0
      ;; ManageStateDriverEnter
      (bind _output_get (Variables|Default|SetDriverPawn PlayerPawn))
      (FortPawn|BindEventtoOnWeaponEquippedDelegate _output_get (EventDispatchers|CreateEvent :!event Custom|OnWeaponEquipped)) ; Check to see if player is changing to a rocket
      (Utilities|FlowControl|Sequence
        (:then_0
          (bind _asplayer_controller (Utilities|Casting|CastToPlayerController (Pawn|GetController (Variables|Default|GetDriverPawn))))
          (Variables|Default|SetDrivingPlayerController _asplayer_controller)
          (CallFunction|CanHitLastDriver)
          (Variables|Default|SetPassengerIsInstigator false)
          (CallFunction|StartDriverRumble)
          (CallFunction|StartDriverCameraShake))
        (:then_1
          ;;; ManageHeadlightsOn_2
          (Utilities|FlowControl|Sequence
            (:then_0
              (TimeOfDay|RegisterforDayPhaseChangedEvent (TimeOfDay|GetContextualTimeOfDayManager) (delegate Custom|SetHeadlightState) false))
            (:then_1
              (CallFunction|SetHeadlightState (TimeOfDay|GetCurrentDayPhase :!class /Script/FortniteGame.FortKismetLibrary)))
            (:then_2
              _))
          (Variables|Camera|SetCameraSpaceForwardDistanceOffset 0.0))
        (:then_2
          (bind _asb_ranged_generic (Utilities|Casting|CastToB_Ranged_Generic (Variables|/Script/FortniteGame.FortPawn|GetCurrentWeapon (Variables|Default|GetDriverPawn))))
          (Default|BindEventtoOnAimDownSightsChanged _asb_ranged_generic (EventDispatchers|CreateEvent :!event Custom|LocalPassengerADS)))))
    (:1
      _)
    (:2
      _)
    (:3
      _))
  ;; ManageStatePassengerEnter
  (bind _output_get_2 (Variables|Default|SetPassengerPawn PlayerPawn))
  (FortPawn|BindEventtoOnWeaponEquippedDelegate _output_get_2 (EventDispatchers|CreateEvent :!event Custom|OnWeaponEquipped)) ; Check to see if player is changing to a rocket
  (Utilities|FlowControl|Sequence
    (:then_0
      (if (== SeatIdx 1)
        (Class|PassengerImpulseWeaponComponent|SetPassengerPawn (Variables|Default|GetImpulseComponent_Passenger1) (Variables|Default|GetPassengerPawn) :!class /Game/Athena/DrivableVehicles/PassengerImpulseWeaponComponent.PassengerImpulseWeaponComponent_C)
        (elif (== SeatIdx 2)
          (Class|PassengerImpulseWeaponComponent|SetPassengerPawn (Variables|Default|GetImpulseComponent_Passenger2) (Variables|Default|GetPassengerPawn) :!class /Game/Athena/DrivableVehicles/PassengerImpulseWeaponComponent.PassengerImpulseWeaponComponent_C)
          (elif (== SeatIdx 3)
            (Class|PassengerImpulseWeaponComponent|SetPassengerPawn (Variables|Default|GetImpulseComponent_Passenger3) (Variables|Default|GetPassengerPawn) :!class /Game/Athena/DrivableVehicles/PassengerImpulseWeaponComponent.PassengerImpulseWeaponComponent_C)))))
    (:then_1
      (bind _asplayer_controller_2 (Utilities|Casting|CastToPlayerController (Pawn|GetController (Variables|Default|GetPassengerPawn))))
      (bind _output_get_3 (Variables|Default|SetPassengerPlayerController _asplayer_controller_2))
      (bind _returnvalue (CameraShakes|StartCameraShake :self (Variables|/Script/Engine.PlayerController|GetPlayerCameraManager _output_get_3) :ShakeClass "/Game/Athena/DrivableVehicles/Speed_CameraShake.Speed_CameraShake_C" :Scale 1.0 :PlaySpace "CameraLocal"))
      (Variables|CameraShake|SetPassengerCameraShake (CameraShakes|ConvLegacyCameraShake _returnvalue)))
    (:then_2
      (bind _asb_ranged_generic_2 (Utilities|Casting|CastToB_Ranged_Generic (Variables|/Script/FortniteGame.FortPawn|GetCurrentWeapon _output_get_2)))
      (Default|BindEventtoOnAimDownSightsChanged _asb_ranged_generic_2 (EventDispatchers|CreateEvent :!event Custom|LocalPassengerADS)))))

(event EventOnCollisionHitEffects
  (params (HitLocation :type Vector%20%28by%20ref%29 :default "0, 0, 0") (HitNormalImpulse :type Vector%20%28by%20ref%29 :default "0, 0, 0") (HitFrictionImpulse :type Vector%20%28by%20ref%29 :default "0, 0, 0") (HitNormal :type Vector%20%28by%20ref%29 :default "0, 0, 0") (HitActor :type Actor%20Object%20Reference) (HitSurfaceType :type EPhysicalSurface%20Enum :default "SurfaceType_Default"))
  (bind (_asfort_pawn _bsuccess) (Utilities|Casting|CastToFortPawn HitActor))
  ;[0 0 0] Audio/FX Collision
  (bind (_outresult _outmagnitude) (Audio|OnCollision (Variables|Default|GetFortCollisionAudioImpacts) HitLocation HitNormalImpulse))
  (if _outresult
    (CallFunction|PlayHapticsonHit _outmagnitude)
    (if (> _outmagnitude 1.5e+04)
      (bind _returnvalue (Niagara|SpawnSystematLocation (select _bsuccess "/Game/Athena/DrivableVehicles/FX/NS_Shopping_Cart_Vehicle_HitPlayer.NS_Shopping_Cart_Vehicle_HitPlayer" "/Game/Athena/DrivableVehicles/FX/NS_ShoppingCart/ns_Shopping_Cart_Vehicle_HitDust.NS_Shopping_Cart_Vehicle_HitDust") HitLocation (Math|Conversions|RotationFromXVector HitNormal)))
      (Effects|Components|ParticleSystem|SetEmitterEnable _returnvalue "Sparks" (> _outmagnitude 4.5e+04)) ; Turn on sparks for heavy hits
    )))

(event Custom|UpdateMovementCosmeticParameters
  ;[0 0 0] Audio Modulation (Movement, Coast, Skid, In-Air, Scrape) + FX
  (if (not (Vehicle|IsAsleep))
    (Utilities|FlowControl|Sequence
      (:then_0
        (bind _returnvalue (Audio|AreAnyListenersWithinRange (Transformation|GetActorLocation) (Variables|Default|GetEngineSoundMaxDistance)))
        (if _returnvalue
          (Utilities|FlowControl|Sequence
            (:then_0
              (Utilities|FlowControl|Sequence
                (:then_0
                  (Class|FortVehicleAudioVoice|SetParam (Variables|Default|GetAudioMovement) (Variables|Vehicle|GetMovementParam)))
                (:then_1
                  (Class|FortVehicleAudioVoice|SetParam (Variables|Default|GetAudioBoost) (Variables|Vehicle|GetBoostParam)))
                (:then_2
                  (Class|FortVehicleAudioVoice|SetParam (Variables|Default|GetAudioHandbrakeSkid) (Variables|Vehicle|GetHandBrakeSkidParam)))
                (:then_3
                  (if (Vehicle|IsHonking)
                    _)
                  (Utilities|FlowControl|DoOnce false
                    (:Completed
                      (Audio|Components|Audio|Play (Variables|Default|GetGolfCart_Horn_Cue) 0.0))))))
            (:then_1
              _)))
        (Utilities|FlowControl|DoOnce false
          (:Completed
            (Class|FortVehicleAudioVoice|SetParam (Variables|Default|GetAudioMovement) 0.0))))
      (:then_1
        (if (<= (Vehicle|GetSignificanceLOD) 1)
          (Utilities|FlowControl|Sequence
            (:then_0
              (Utilities|FlowControl|Sequence
                (:then_0
                  (Class|FortVehicleAudioVoice|SetParam (Variables|Default|GetAudioBattery) (Variables|Vehicle|GetBatteryParam)))
                (:then_1
                  (Class|FortVehicleAudioVoice|SetParam (Variables|Default|GetAudioSkid) (Variables|Vehicle|GetSkiddingAmountParam)))
                (:then_2
                  (Class|FortVehicleAudioVoice|SetParam (Variables|Default|GetAudioInAir) (Variables|Vehicle|GetInAirParam)))
                (:then_3
                  (Class|FortVehicleAudioVoice|SetParam (Variables|Default|GetAudioScrape) (Variables|Vehicle|GetScrapingAmountParam)))
                (:then_4
                  (Class|FortVehicleAudioVoice|SetParam (Variables|Default|GetAudioWind) (Variables|Vehicle|GetAudioWindParam)))))
            (:then_1
              _)))
        (Utilities|FlowControl|DoOnce false
          (:Completed
            (Class|FortVehicleAudioVoice|SetParam (Variables|Default|GetAudioInAir) 0.0)))))))

(event EventPawnHitByVehicle
  (params (Pawn :type Fort%20Pawn%20Object%20Reference) (ImpactPoint :type Vector%20%28by%20ref%29 :default "0, 0, 0") (ImpactNormal :type Vector%20%28by%20ref%29 :default "0, 0, 0") (LaunchVector :type Vector :default "0, 0, 0") (bTeleportPawn :type Boolean) (TeleportPawnDistance :type Float%20%28single-precision%29))
  ;[1 1 1] EVENT HIT -> Launch Player hit by Cart
  (if (> (ScalableFloat|GetValueatLevel (Variables|Default|GetCanLaunchPlayer)) 0.0)
    (bind _asfort_player_pawn_athena (Utilities|Casting|CastToFortPlayerPawnAthena Pawn)
      (:then
        (if (Parachute|IsSkydiving _asfort_player_pawn_athena)
          _
          (elif (Math|Boolean|ORBoolean (Utilities|Array|ContainsItem (Variables|Default|GetPassengers) _asfort_player_pawn_athena) (Utilities|Equal(Object) _asfort_player_pawn_athena (Variables|Default|GetLastDriver)) (Utilities|Equal(Object) _asfort_player_pawn_athena (Variables|Default|GetLastHitPlayer)))
            (if (and (Utilities|Equal(Object) (Variables|Default|GetLastDriver) _asfort_player_pawn_athena) (Variables|Default|GetIgnoreHitLastDriver))
              _
              (else
                _))))
        (Character|LaunchCharacter _asfort_player_pawn_athena LaunchVector true true) ; Launch player based on cart velocity
        (for _array_element (Vehicle|GetAllPassengers)
          (GameplayCue|ExecuteGameplayCue%28Burst%29Local _array_element (Variables|Default|GetHitPlayerCue) (Variables|Default|GetGCParamsEmpty))
          (Variables|Default|SetLastHitPlayer _asfort_player_pawn_athena)
          (if (Vehicle|HasDriver)
            (Class|FortPlayerPawnAthena|SetFallInstigator :self _asfort_player_pawn_athena :NewFallInstigator (Game|GetInstigatorController (Variables|Default|GetDriverPawn)))
            (elif (Variables|Default|GetPassengerIsInstigator)
              (Class|FortPlayerPawnAthena|SetFallInstigator :self _asfort_player_pawn_athena :NewFallInstigator (Game|GetInstigatorController (Utilities|Array|Get(acopy) (Variables|Default|GetPassengers))))
              (else
                (Class|FortPlayerPawnAthena|SetFallInstigator :self _asfort_player_pawn_athena :NewFallInstigator (Game|GetInstigatorController (Variables|Default|GetLastDriver))))))
          (Utilities|FlowControl|Delay 0.2)
          (Variables|Default|SetLastHitPlayer 0)))
      (:CastFailed
        (bind _asfort_aipawn (Utilities|Casting|CastToFortAIPawn Pawn))
        (Components|Movement|StopMovementImmediately (Variables|/Script/Engine.Character|GetCharacterMovement _asfort_aipawn))
        (Character|LaunchCharacter _asfort_aipawn LaunchVector true true) ; Launch player based on cart velocity
        (Networking|SwitchHasAuthority
          (:Remote
            (for _array_element_2 (Vehicle|GetAllPassengers)
              (GameplayCue|ExecuteGameplayCue%28Burst%29Local _array_element_2 (Variables|Default|GetHitFiendCue) (Variables|Default|GetGCParamsEmpty)))))))))

(event EventOnStopTick
  ;[0 0 0] Audio Modulation (Movement, Coast, Skid, In-Air, Scrape) + FX
  (Class|FortVehicleAudioVoice|StopAllAudio (Variables|Default|GetAudioMovement)))

(event Custom|StartBoostRumble
  ;[1 1 1] Rumble
  (Game|Feedback|PlayDynamicForceFeedback (Variables|Default|GetDrivingPlayerController) (Variables|Default|GetBoostRumbleValue) -1 true false true false))

(event Custom|UpdateBoostRumble
  ;[1 1 1] Rumble
  (Game|Feedback|PlayDynamicForceFeedback (Variables|Default|GetDrivingPlayerController) (Variables|Default|GetBoostRumbleValue) -1 true false true false))

(event Custom|EventOnHealthChanged
  ;[1 1 1] Update cart material based on remaining health %
  (CallFunction|UpdateDamageState (Math|Float|Lerp 1.0 0.0 (/ (Class|FortVehicleInterface|GetHealth) (Class|FortVehicleInterface|GetMaxHealth)))))

(event Custom|BoostRumbleEnd
  ;[1 1 1] Rumble
  (Game|Feedback|PlayDynamicForceFeedback (Variables|Default|GetDrivingPlayerController) (Variables|Default|GetBoostRumbleValue) -1 true false true false))

(event Custom|Update%20Damage%20State
  (params (Damage :type Float%20%28double-precision%29))
  ;[1 1 1] Update cart material based on remaining health %
  (CallFunction|SendDamageValuetoMaterials Damage))

(event Custom|OnPassengerFiredImpulseWeapon_Event)

(event Custom|LocalPassengerADS
  (params (IsADS :type Boolean))
  ;[1 1 1] PassengerADS switches the vehicle material to a special masked one so it won't occlude while aiming
  (Variables|Rumble|SetbLocalPlayerADS IsADS)
  (CallFunction|ChangeOutMaterials))

(event EventBoostBegin
  ;[1 1 1] Boost Effects
  (Utilities|FlowControl|Sequence
    (:then_0
      (CallFunction|StartBoostRumble)
      (Variables|Rumble|SetbBoostCameraActive true)
      (Components|Timeline|SetPlayRate (Variables|GolfCartVehicleSK|GetBoostTimeline) 1.0)
      (CallFunction|SendBoostTimetoMaterials 9999999.0)
      (bind (_direction _track1 _rumble) (Timeline|BoostTimeline Play)
        (:Update
          (Variables|Camera|SetCameraFOVOffset (Math|Float|Lerp 0.0 2e+01 _track1))
          (Variables|Default|SetBoostRumbleValue _rumble)
          (CallFunction|UpdateBoostRumble)
          (switch Utilities|FlowControl|Switch|SwitchonETimelineDirection _direction
            (:Forward
              (Rendering|Material|SetScalarParameterValueonMaterials (Variables|Vehicle|GetSkeletalMesh) "BoostGlow" (Math|Float|Clamp(Float) (* _track1 1.0))))))
        (:Finished
          (switch Utilities|FlowControl|Switch|SwitchonETimelineDirection _direction
            (:Forward
              (CallFunction|BoostCameraReverse))
            (:Backward
              (Variables|Rumble|SetbBoostCameraActive false)
              (CallFunction|BoostRumbleEnd))))))
    (:then_1
      (Components|Activation|Activate (Variables|Default|GetP_GolfCart_Boost_Flames_r) (Variables|Default|GetResetBoostFX))
      (Audio|SpawnSoundAttached (Variables|Default|GetOnBoostSound) (Variables|Vehicle|GetSkeletalMesh))
      (bind _boostincreaseaudiocomp (Variables|Default|GetBoostIncreaseAudioComp)
        (:then
          (Audio|Components|Audio|FadeOut _boostincreaseaudiocomp 0.2 0.0 "Linear" :!class /Script/Engine.AudioComponent))
        (:else)))))

(event EventEBrakeBegin)

(event Custom|ChangeOutMaterials
  ;[1 1 1] Update materials because the driver changed or ADS changed
  (Class|FortVehicleInterface|ApplyMaterialandUseWrapifPossible (select (Variables|Rumble|GetbLocalPlayerADS) "/Game/Athena/DrivableVehicles/Golf_Cart/Materials/MI_Golf_Cart_Base_Masked.MI_Golf_Cart_Base_Masked" "/Game/Athena/DrivableVehicles/Golf_Cart/Materials/MI_Golf_Cart_Base.MI_Golf_Cart_Base") 1)
  (Class|FortVehicleInterface|ApplyMaterialandUseWrapifPossible (select (Variables|Rumble|GetbLocalPlayerADS) "/Game/Athena/DrivableVehicles/Golf_Cart/Golf_Cart_Kit_A/Materials/MI_Golf_Cart_Kit_MasterA_Masked.MI_Golf_Cart_Kit_MasterA_Masked" "/Game/Athena/DrivableVehicles/Golf_Cart/Golf_Cart_Kit_A/Materials/MI_Golf_Cart_Kit_MasterA.MI_Golf_Cart_Kit_MasterA"))
  (CallFunction|SendDamageValuetoMaterials (Variables|Default|GetSavedDamageValueForMID))
  (CallFunction|SendHeadlightValuetoMaterials (Variables|Default|GetSavedHeadlightValueForMID))
  (CallFunction|SendBoostTimetoMaterials (Variables|Default|GetBoostTimeValueForMID)))

(event EventEBrakeEnd)

(event EventK2_ApplyCosmeticWrap
  (params (LoadedWrap :type Athena%20Item%20Wrap%20Definition%20Object%20Reference))
  ;[1 1 1] Update materials because the driver changed or ADS changed
  (Class|FortVehicleInterface|ApplyMaterialandUseWrapifPossible (select (Variables|Rumble|GetbLocalPlayerADS) "/Game/Athena/DrivableVehicles/Golf_Cart/Materials/MI_Golf_Cart_Base_Masked.MI_Golf_Cart_Base_Masked" "/Game/Athena/DrivableVehicles/Golf_Cart/Materials/MI_Golf_Cart_Base.MI_Golf_Cart_Base") 1)
  (Class|FortVehicleInterface|ApplyMaterialandUseWrapifPossible (select (Variables|Rumble|GetbLocalPlayerADS) "/Game/Athena/DrivableVehicles/Golf_Cart/Golf_Cart_Kit_A/Materials/MI_Golf_Cart_Kit_MasterA_Masked.MI_Golf_Cart_Kit_MasterA_Masked" "/Game/Athena/DrivableVehicles/Golf_Cart/Golf_Cart_Kit_A/Materials/MI_Golf_Cart_Kit_MasterA.MI_Golf_Cart_Kit_MasterA"))
  (CallFunction|SendDamageValuetoMaterials (Variables|Default|GetSavedDamageValueForMID))
  (CallFunction|SendHeadlightValuetoMaterials (Variables|Default|GetSavedHeadlightValueForMID))
  (CallFunction|SendBoostTimetoMaterials (Variables|Default|GetBoostTimeValueForMID)))

(event Custom|ReverseBoostTimeline
  ;[1 1 1] Boost Effects
  (Timeline|BoostTimeline Reverse))

(event EventPowerSlideBegin
  ;[1 1 1] Cart Trails
  (CallFunction|SetSkidActive true))

(event EventPowerSlideEnd
  ;[1 1 1] Cart Trails
  (CallFunction|SetSkidActive false))

(event Custom|BoostCameraReverse
  ;[1 1 1] Boost Effects
  (if (Variables|Rumble|GetbBoostCameraActive)
    (Components|Timeline|SetPlayRate (Variables|GolfCartVehicleSK|GetBoostTimeline) 1.0)
    (CallFunction|ReverseBoostTimeline)))

(event EventBoostIncrease
  (params (AccumulationStepIndex :type Integer))
  ;[1 1 1] Boost Effects
  ;;[1 1 1] Feedback for when the boost changes color
  (bind _returnvalue (Audio|SpawnSoundAttached (Utilities|Array|Get(acopy) (Variables|Default|GetOnBoostIncreaseSounds) (- AccumulationStepIndex 1)) (Variables|Vehicle|GetSkeletalMesh)))
  (Variables|Default|SetBoostIncreaseAudioComp _returnvalue))

(event EventEndPlay
  (params (EndPlayReason :type EEndPlayReason%20Enum :default "Destroyed"))
  ;[1 1 1] EVENT END PLAY
  (Utilities|IsValid (Variables|Default|GetDriverPawn)))

(event EventBoostEnd
  ;[1 1 1] Boost Effects
  (Utilities|FlowControl|Sequence
    (:then_0
      (if (Variables|Rumble|GetbBoostCameraActive)
        (Components|Timeline|SetPlayRate (Variables|GolfCartVehicleSK|GetBoostTimeline) 1.0)
        (CallFunction|ReverseBoostTimeline)))
    (:then_1
      ;;[1 1 1] Boost FX
      (Components|Activation|Deactivate (Variables|Default|GetP_GolfCart_Boost_Flames_l))
      (Variables|Default|SetResetBoostFX false)
      (CallFunction|SendBoostTimetoMaterials (Utilities|Time|GetGameTimeinSeconds)))))

(event Vehicle|EventOnSpringCompression
  (bind _drivingplayercontroller (Variables|Default|GetDrivingPlayerController))
  ;[1 1 1] OnSpringCompression - play force feedback / audio
  (CallFunction|HandleSpringCompressionAudio)
  ;; PlayHapticsOnSpringCompression
  (Utilities|IsValid _drivingplayercontroller
    (:Is%20Valid
      (Game|Feedback|ClientPlayForceFeedback _drivingplayercontroller "/Game/Weapons/ForceFeedback/FF_ShoppingCartBump.FF_ShoppingCartBump" "None" false false false))))

(event Custom|Handle%20Spring%20Compression%20Audio
  ;[1 1 1] Spring Compression Audio
  (Audio|OnCollision (Variables|Default|GetFortCollisionAudioSuspension) (Transformation|GetActorLocation) (Math|Vector|MakeVector (Math|Float|float+float (Vehicle|GetSpringLengthRatio 0) (Vehicle|GetSpringLengthRatio 1) (Vehicle|GetSpringLengthRatio 2) (Vehicle|GetSpringLengthRatio 3)))))

(event EventBeginPlay
  ;[1 1 1] EVENT BEGIN PLAY
  (Variables|Default|SetResetBoostFX true)
  (Variables|Default|SetIgnoreHitLastDriver true)
  (Variables|Default|SetPassengerIsInstigator true)
  (Default|AssignOnPassengerFiredImpulseWeapon (Variables|Default|GetImpulseComponent_Passenger1) (delegate Custom|OnPassengerFiredImpulseWeapon))
  (CallFunction|CacheParticleComponentPointers (Variables|Default|GetP_GolfCart_Vehicle_WheelDust) (Variables|Default|GetP_GolfCart_Vehicle_WheelSlide) (Variables|Default|GetTrailLeft) (Variables|Default|GetTrailRight))
  (CallFunction|SetPropDestructionParams (Utilities|Array|MakeArray "/Script/FortniteGame.BuildingProp" "/Game/Building/ActorBlueprints/Containers/ParentBuildingContainerBlueprint.ParentBuildingContainerBlueprint_C") (Utilities|Array|MakeArray) (Utilities|Array|MakeArray "/Game/Building/EditModePatterns/Wall/EMP_Wall_HalfWallS.EMP_Wall_HalfWallS") (Utilities|Array|MakeArray))
  (Health|BindEventtoOnHealthChanged (EventDispatchers|CreateEvent :!event Custom|EventOnHealthChanged) :!class /Script/FortniteGame.FortAthenaVehicle)
  ;; If server, nuke audio and fx
  (if (Networking|IsDedicatedServer)
    (Components|DestroyComponent (Variables|Default|GetP_GolfCart_Vehicle_WheelDust))))

(event Custom|OnPassengerFiredImpulseWeapon
  ;[1 1 1] EVENT BEGIN PLAY
  (Variables|Default|SetPassengerIsInstigator true))

(event EventOnDeathPlayEffects
  (params (Damage :type Float%20%28single-precision%29) (DamageTags :type Gameplay%20Tag%20Container%20Structure%20%28by%20ref%29) (Momentum :type Vector :default "0, 0, 0") (HitInfo :type Hit%20Result%20Structure%20%28by%20ref%29) (InstigatedBy :type Fort%20Pawn%20Object%20Reference) (DamageCauser :type Actor%20Object%20Reference) (EffectContext :type Gameplay%20Effect%20Context%20Handle%20Structure))
  ;[1 1 1] Play Death Effects
  (bind _returnvalue (Effects|SpawnEmitteratLocation :EmitterTemplate "/Game/Athena/DrivableVehicles/FX/P_Golfcart_Vehicle_Destroyed.P_Golfcart_Vehicle_Destroyed" :Location (Transformation|GetActorLocation) :Scale "1.000000,1.000000,1.000000" :bAutoDestroy true :PoolingMethod "AutoRelease" :bAutoActivateSystem true))
  (Effects|Components|ParticleSystem|SetVectorParameter _returnvalue "Velocity" (Transformation|GetVelocity))
  (Audio|PlaySoundatLocation (Variables|Default|GetOnDeathSound) (Transformation|GetActorLocation))
  (Development|SetHiddeninGame (Variables|Vehicle|GetSkeletalMesh) true false))

(event EventOnHoldExitStarted
  (params (ExitingPawn :type Fort%20Pawn%20Object%20Reference) (ExitDuration :type Float%20%28single-precision%29))
  ;[1 1 1] EVENT -> OnHoldExitStarted/Stopped
  (Ability|SendGameplayEventtoActor ExitingPawn "(TagName=\"Athena.Vehicle.Biplane.OnHoldExitStarted\")" (Utilities|Struct|MakeGameplayEventData :ContextHandle "()" :InstigatorTags "(GameplayTags=)" :TargetTags "(GameplayTags=)" :EventMagnitude ExitDuration :TargetData "()")))

(event EventOnHoldExitStopped
  (params (ExitingPawn :type Fort%20Pawn%20Object%20Reference))
  ;[1 1 1] EVENT -> OnHoldExitStarted/Stopped
  (Ability|SendGameplayEventtoActor ExitingPawn "(TagName=\"Athena.Vehicle.Biplane.OnHoldExitStopped\")"))

(event Custom|RemoteClientBounceEffects
  (bind _bounceoverlappawn (Variables|Default|GetBounceOverlapPawn))
  (bind _returnvalue (Transformation|GetActorLocation _bounceoverlappawn))
  (bind _bounceoverlapcontroller (Variables|Default|GetBounceOverlapController))
  (Audio|PlaySoundatLocation (Variables|Default|GetOnRoofBounceSound) _returnvalue)
  (Effects|SpawnEmitteratLocation "/Game/Effects/Fort_Effects/Gameplay/SmokeDustExplosions/Particles/P_Player_Bounce_Up.P_Player_Bounce_Up" _returnvalue)
  (Utilities|IsValid _bounceoverlapcontroller
    (:Is%20Valid
      (Game|Feedback|LocalPlayForceFeedback _bounceoverlapcontroller (Variables|Default|GetLaunchForceFeedback) false "TireLaunch"))
    (:Is%20Not%20Valid
      _))
  (Variables|Default|SetBounceOverlapPawn 0)
  (Variables|Default|SetBounceOverlapController 0)
  (Animation|Montage|PlayMontage (Variables|Vehicle|GetSkeletalMesh) "/Game/Animation/Game/MainPlayer/Vehicles/GolfCart/PartsAnims/Roof_Bounce_M.Roof_Bounce_M")
  (if (Vehicle|IsAsleep)
    (Components|Tick|SetComponentTickEnabled (Variables|Vehicle|GetSkeletalMesh) true)
    (Utilities|Time|SetTimerbyEvent (delegate Custom|DisableSkelmeshTickAfterBounce) 1.0)))

(event Custom|DisableSkelmeshTickAfterBounce
  (if (Vehicle|IsAsleep)
    (Components|Tick|SetComponentTickEnabled (Variables|Vehicle|GetSkeletalMesh) false)))

(event OnComponentBeginOverlap(RoofCollision)
  (params (OverlappedComponent :type Primitive%20Component%20Object%20Reference) (OtherActor :type Actor%20Object%20Reference) (OtherComp :type Primitive%20Component%20Object%20Reference) (OtherBodyIndex :type Integer) (bFromSweep :type Boolean) (SweepResult :type Hit%20Result%20Structure%20%28by%20ref%29))
  (bind _bounceoverlappawn (Variables|Default|GetBounceOverlapPawn))
  (bind _returnvalue (Transformation|GetActorLocation _bounceoverlappawn))
  (bind (_bblockinghit _binitialoverlap _time _distance _location _impactpoint _normal) (Collision|BreakHitResult SweepResult))
  (bind _vehiclehitactor (Variables|Default|GetVehicleHitActor))
  (bind _bounceoverlapcontroller (Variables|Default|GetBounceOverlapController))
  (bind _output_get (Variables|Default|SetVehicleHitActor OtherActor))
  (bind _asfort_player_pawn_athena (Utilities|Casting|CastToFortPlayerPawnAthena _output_get)
    (:then
      (Utilities|IsA(soft) OtherActor "/Game/Athena/Items/Weapons/Abilities/RCRocket/B_PrjPawn_Athena_RCRocket.B_PrjPawn_Athena_RCRocket_C"
        (:False
          (if (> (.z (Math|Transform|InverseTransformLocation (Transformation|GetSocketTransform (Variables|Vehicle|GetSkeletalMesh) "RoofBounceSocket" "RTS_World") _impactpoint)) 0.0)
            (Variables|Default|SetBounceOverlapPawn _asfort_player_pawn_athena)
            (if (not (Math|Boolean|ORBoolean (GameplayTags|HasMatchingGameplayTag (Ability|GetAbilitySystemComponent (Variables|Default|GetBounceOverlapPawn)) "(TagName=\"Gameplay.Action.Player.Launched\")") (Variables|/Script/FortniteGame.FortPawn|GetbIsDBNO (Variables|Default|GetBounceOverlapPawn)) (Parachute|IsSkydiving (Variables|Default|GetBounceOverlapPawn)) (not (Vehicle|CanContainPlayers))))
              (bind _asfort_player_controller (Utilities|Casting|CastToFortPlayerController (Pawn|GetController (Variables|Default|GetBounceOverlapPawn)))
                (:then
                  (Variables|Default|SetBounceOverlapController _asfort_player_controller)
                  ;[1 1 1] Ceiling check
                  (bind _output_get_2 (Variables|Default|SetBounceIgnoreList (Variables|Default|GetPassengers)))
                  (Utilities|Array|Add _output_get_2 (Variables|Default|GetBounceOverlapPawn))
                  (Variables|Default|SetBounceVector "0, 0,1.000000")
                  (bind (_outhit _returnvalue_2) (Collision|CapsuleTraceByChannel (Math|Vector|MakeVector (.x (Transformation|GetWorldLocation (Variables|Transformation|GetRootComponent))) (.y (Transformation|GetWorldLocation (Variables|Transformation|GetRootComponent))) (+ (.z (Transformation|GetWorldLocation (Variables|Transformation|GetRootComponent))) 200)) (Math|Vector|MakeVector (.x (Transformation|GetWorldLocation (Variables|Transformation|GetRootComponent))) (.y (Transformation|GetWorldLocation (Variables|Transformation|GetRootComponent))) (+ (.z (Transformation|GetWorldLocation (Variables|Transformation|GetRootComponent))) 340)) 50 100 "TraceTypeQuery9" false (Variables|Default|GetBounceIgnoreList)))
                  (Variables|Default|SetBounceHeight (select _returnvalue_2 (Variables|Default|GetObstructedBounceVelocity) (Variables|Default|GetUnobstructedBounceVelocity)))
                  (RemoteControlPawn|ForceDetachfromRemoteControlledPawn (Variables|Default|GetBounceOverlapPawn))
                  (Character|LaunchCharacter (Variables|Default|GetBounceOverlapPawn) (Math|Vector|MakeVector 0.0 0.0 (Variables|Default|GetBounceHeight)) false true)
                  (if (Networking|IsDedicatedServer)
                    (GameplayEffects|ApplyGameplayEffectToSelf :self (Ability|GetAbilitySystemComponent (Variables|Default|GetBounceOverlapPawn)) :GameplayEffectClass (Variables|Default|GetGE_PlayerLaunched) :Level 0.0)
                    (else
                      (Audio|PlaySoundatLocation (Variables|Default|GetOnRoofBounceSound) _returnvalue)
                      (Effects|SpawnEmitteratLocation "/Game/Effects/Fort_Effects/Gameplay/SmokeDustExplosions/Particles/P_Player_Bounce_Up.P_Player_Bounce_Up" _returnvalue)
                      (Utilities|IsValid _bounceoverlapcontroller
                        (:Is%20Valid
                          (Game|Feedback|LocalPlayForceFeedback _bounceoverlapcontroller (Variables|Default|GetLaunchForceFeedback) false "TireLaunch"))
                        (:Is%20Not%20Valid
                          _))
                      (Variables|Default|SetBounceOverlapPawn 0)
                      (Variables|Default|SetBounceOverlapController 0)
                      (Animation|Montage|PlayMontage (Variables|Vehicle|GetSkeletalMesh) "/Game/Animation/Game/MainPlayer/Vehicles/GolfCart/PartsAnims/Roof_Bounce_M.Roof_Bounce_M")
                      (if (Vehicle|IsAsleep)
                        (Components|Tick|SetComponentTickEnabled (Variables|Vehicle|GetSkeletalMesh) true)
                        (Utilities|Time|SetTimerbyEvent (delegate Custom|DisableSkelmeshTickAfterBounce) 1.0)))))
                (:CastFailed
                  (CallFunction|RemoteClientBounceEffects))))))))
    (:CastFailed
      (if (Networking|IsDedicatedServer)
        (if (Math|Boolean|ANDBoolean (Variables|Actor|GetbCanBeDamaged (Variables|Default|GetVehicleHitActor)) (Class|FortVehicleInterface|HasPermissiontoAlterBuildings) (and (<= (Math|Vector|DotProduct (Math|Vector|Normalize (Vehicle|GetPreviousLinearVelocity)) _normal) (Variables|Default|GetDestructionAngle)) (>= (Vehicle|GetPreviousForwardSpeedKmh) 39.0)))
          (bind _asbuilding_wall (Utilities|Casting|CastToBuildingWall _vehiclehitactor)
            (:then
              (if (BuildingActor|IsPlayerBuilt _asbuilding_wall)
                (switch Utilities|FlowControl|Switch|SwitchonEFortResourceType (Variables|/Script/FortniteGame.BuildingSMActor|GetResourceType _asbuilding_wall)
                  (:Wood
                    (CallFunction|DestroyorDamageBuildingPiece _asbuilding_wall true "/Game/Athena/DrivableVehicles/Golf_Cart/Golf_Cart_Base/Blueprints/GE_VehicleDamageWalls300.GE_VehicleDamageWalls300_C"))
                  (:Stone
                    _)
                  (:Metal
                    _))
                (CallFunction|DestroyorDamageBuildingPiece _asbuilding_wall true "/Game/Athena/DrivableVehicles/Golf_Cart/Golf_Cart_Base/Blueprints/GE_VehicleDamageWalls200.GE_VehicleDamageWalls200_C")
                (else
                  (CallFunction|DestroyorDamageBuildingPiece _asbuilding_wall true "/Game/Athena/DrivableVehicles/Golf_Cart/Golf_Cart_Base/Blueprints/GE_VehicleDamageWalls300.GE_VehicleDamageWalls300_C"))))
            (:CastFailed
              (bind _asbuilding_prop (Utilities|Casting|CastToBuildingProp _vehiclehitactor)
                (:then
                  (if (not (Actor|ActorHasTag _asbuilding_prop "NoVehicleDamage"))
                    (CallFunction|DestroyorDamageBuildingPiece _asbuilding_prop true "/Game/Athena/DrivableVehicles/Golf_Cart/Golf_Cart_Base/Blueprints/GE_VehicleDamageWalls300.GE_VehicleDamageWalls300_C")))
                (:CastFailed
                  (bind _asathena_prop_parent_building_container_blueprint (Utilities|Casting|CastToAthena_Prop_ParentBuildingContainerBlueprint _vehiclehitactor)
                    (:then
                      (CallFunction|DestroyorDamageBuildingPiece _asathena_prop_parent_building_container_blueprint true "/Game/Athena/DrivableVehicles/Golf_Cart/Golf_Cart_Base/Blueprints/GE_VehicleDamageWalls300.GE_VehicleDamageWalls300_C"))
                    (:CastFailed
                      (bind _asfort_athena_supply_drop (Utilities|Casting|CastToFortAthenaSupplyDrop (Variables|Default|GetVehicleHitActor)))
                      (CallFunction|DestroyorDamageBuildingPiece _asfort_athena_supply_drop true "/Game/Athena/DrivableVehicles/Golf_Cart/Golf_Cart_Base/Blueprints/GE_VehicleDamageWalls200.GE_VehicleDamageWalls200_C" 0.0))))))))))))

(event Custom|DestroyCooldown)

(fn AttachWeaponToCart
  (params (Selection :type Integer))
  (bind (_returnvalue_roll _returnvalue_pitch _returnvalue_yaw) (Math|Rotator|BreakRotator (Transformation|GetActorRotation (Variables|Default|GetFortPickup))))
  (bind _skeletalmesh (Variables|Vehicle|GetSkeletalMesh))
  (bind _fortpickup (Variables|Default|GetFortPickup))
  (switch int Selection
    (:Default
      _)
    (:0
      _)
    (:1
      (Transformation|AttachActorToComponent _fortpickup _skeletalmesh "Weapon2" "SnapToTarget" "SnapToTarget" "SnapToTarget" true)
      (Transformation|SetActorRotation (Variables|Default|GetFortPickup) (Math|Rotator|MakeRotator (Math|Random|RandomFloatinRange (+ _returnvalue_roll -1e+02) (+ _returnvalue_roll -8e+01)) (Math|Random|RandomFloatinRange (+ _returnvalue_pitch -5.0) (+ _returnvalue_pitch 5.0)) (Math|Random|RandomFloatinRange (+ _returnvalue_yaw -1e+02) (+ _returnvalue_yaw -8e+01))) false :!label _label))
    (:2
      (Transformation|AttachActorToComponent _fortpickup _skeletalmesh "Weapon3" "SnapToTarget" "SnapToTarget" "SnapToTarget" true)
      (goto _label))
    (:3
      (Transformation|AttachActorToComponent _fortpickup _skeletalmesh "Weapon4" "SnapToTarget" "SnapToTarget" "SnapToTarget" true)
      (goto _label))
    (:4
      (Transformation|AttachActorToComponent _fortpickup _skeletalmesh "Weapon5" "SnapToTarget" "SnapToTarget" "SnapToTarget" true)
      (goto _label)))
  (Transformation|AttachActorToComponent _fortpickup _skeletalmesh "Weapon1" "SnapToTarget" "SnapToTarget" "SnapToTarget" true)
  (goto _label))

(fn DestroyOrDamageBuildingPiece
  (params (BuildingActor :type Building%20Actor%20Object%20Reference) (CheckForHealth :type Boolean) (GameplayEffectClass :type Gameplay%20Effect%20Class%20Reference) (DestroyThreshold :type Float%20%28double-precision%29 :default 3e+02))
  (if (Variables|Actor|GetbCanBeDamaged BuildingActor)
    (if CheckForHealth
      (if (not (< (Health|GetHealth BuildingActor :!class /Script/FortniteGame.BuildingActor) (CallFunction|GetModifiedDamageforActor DestroyThreshold BuildingActor)))
        (GameplayEffects|ApplyGameplayEffectToTarget (Ability|GetAbilitySystemComponent (Variables|Default|GetDriverPawn)) GameplayEffectClass (Ability|GetAbilitySystemComponent (Variables|Default|GetVehicleHitActor)) 0.0 (GameplayEffects|MakeEffectContext (Ability|GetAbilitySystemComponent (Variables|Default|GetDriverPawn))))))
    (Actor|DestroyActor (Variables|Default|GetVehicleHitActor))))

(fn OnRep_CorrectiveImpulse)

(fn PlayHapticsOnHit
  (params (HitMagnitude :type Float%20%28double-precision%29))
  (bind _passengerplayercontroller (Variables|Default|GetPassengerPlayerController))
  (bind _drivingplayercontroller (Variables|Default|GetDrivingPlayerController))
  (bind _validplayercontrollers (LocalVariables|GetValidPlayerControllers))
  (Utilities|FlowControl|Sequence
    (:then_0
      (Utilities|IsValid _drivingplayercontroller
        (:Is%20Valid
          (Utilities|Array|Add (LocalVariables|GetValidPlayerControllers) _drivingplayercontroller))))
    (:then_1
      (Utilities|IsValid _passengerplayercontroller
        (:Is%20Valid
          (Utilities|Array|Add (LocalVariables|GetValidPlayerControllers) _passengerplayercontroller))))
    (:then_2
      (if (> (Utilities|Array|Length (LocalVariables|GetValidPlayerControllers)) 0)
        (if (> (Variables|GetHitMagnitude) 1e+04)
          (Utilities|FlowControl|Sequence
            (:then_0
              (LocalVariables|SetCameraShakeMagnitude (Math|Float|MapRangeClamped (Variables|GetHitMagnitude) 1.5e+04 5e+04 0.1 1.0))
              (Utilities|IsValid (Variables|Default|GetDrivingPlayerController)
                (:Is%20Valid
                  (Game|Feedback|ClientStartCameraShake :self (LocalVariables|GetValidPlayerControllers) :Shake "/Game/Athena/DrivableVehicles/CameraShake_HitImpact_GolfCart.CameraShake_HitImpact_GolfCart_C" :Scale (LocalVariables|GetCameraShakeMagnitude) :PlaySpace "CameraLocal")
                  (if (> (Variables|GetHitMagnitude) 2e+04)
                    (Game|Feedback|ClientPlayForceFeedback _validplayercontrollers "/Game/Weapons/ForceFeedback/FF_Generic_Impact_Near.FF_Generic_Impact_Near" "None" false false false)
                    (else
                      (Game|Feedback|ClientPlayForceFeedback _validplayercontrollers "/Game/Weapons/ForceFeedback/FF_Generic_Build.FF_Generic_Build" "None" false false false))))))
            (:then_1
              (if (< (Math|Float|Absolute(Float) (Math|Vector|DotProduct (Math|Vector|Normalize (Vehicle|GetPreviousLinearVelocity)) "0, 0,1.000000")) 0.5)
                (bind _output_get (LocalVariables|SetHitDelta (- (Variables|Vehicle|GetCachedSpeed) (Math|Vector|VectorLength (Transformation|GetComponentVelocity (Variables|Vehicle|GetSkeletalMesh))))))
                (if (> _output_get (LocalVariables|GetCameraPushVelocityDelta))
                  (Game|Feedback|ClientStartCameraShake (LocalVariables|GetValidPlayerControllers) "/Game/Athena/DrivableVehicles/CameraShake_BigHit.CameraShake_BigHit_C" (Math|Float|MapRangeClamped (LocalVariables|GetHitDelta) 3e+02 9e+02 0.5 1.0) "UserDefined" (Math|Rotator|MakeRotfromX (Vehicle|GetPreviousLinearVelocity))))))))))))

(fn SendBoostTimeToMaterials
  (params (ParameterValue :type Float%20%28double-precision%29))
  (bind _output_get (Variables|Default|SetBoostTimeValueForMID ParameterValue))
  (Rendering|Material|SetScalarParameterValueonMaterials (Variables|Vehicle|GetSkeletalMesh) "BoostTime" _output_get))

(fn SendDamageValueToMaterials
  (params (ParameterValue :type Float%20%28double-precision%29))
  (bind _output_get (Variables|Default|SetSavedDamageValueForMID ParameterValue))
  (Rendering|Material|SetScalarParameterValueonMaterials (Variables|Vehicle|GetSkeletalMesh) "Damage" _output_get))

(fn SendHeadlightValueToMaterials
  (params (ParameterValue :type Float%20%28double-precision%29))
  (bind _output_get (Variables|Default|SetSavedHeadlightValueForMID ParameterValue))
  (Rendering|Material|SetScalarParameterValueonMaterials (Variables|Vehicle|GetSkeletalMesh) "HeadLights" _output_get))

(fn ConstructionScript
  (Variables|FortAthenaVehicle|SetOverlapComponent (Variables|Default|GetOverlapVolume)))
