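# Simple Default Lit material named black, stored with the drivable Golf Cart base vehicle's materials.
# Its graph holds a single near-black vector parameter, with no textures, math nodes or visible wiring to a material output.
# Usage flags enable it on skeletal meshes, morph targets and mesh deformers.
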Material /Game/Athena/DrivableVehicles/Golf_Cart/Golf_Cart_Base/Materials/black
graph {
 properties {
  bUsedWithMeshDeformer=true
  bUsedWithMorphTargets=true
  bUsedWithSkeletalMesh=true
  pixelDepthOffsetMode=PDOM_Legacy
  usedShadingModels=Default Lit
 }
 nodes [VectorParameter() {defaultValue=<rgba 0.0196 0.0235 0.0235 0>}]
}
