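# Physics asset for the drivable Golf Cart vehicle, providing collision shapes for what the shape names indicate are the back-left and back-right passengers.
# A single body on the frame bone holds two matching capsules, BLPassenger and BRPassenger, set side by side at the same height.
# It has no other bodies and no constraints, so this asset defines no chassis or wheel shapes and no jointed parts.
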PhysicsAsset /Game/Athena/DrivableVehicles/Golf_Cart/Golf_Cart_Base/Mesh/PA_Golf_Cart_Physics
bodies {frame [
  {shapeType=Capsule, length=78.4643, radius=37.4095, shapeName=BLPassenger, center=<-133.7294 -30.5539 161>}
  {shapeType=Capsule, length=78.4643, radius=37.4095, shapeName=BRPassenger, center=<-134.1207 32.2282 161>}
 ]}
constraints []
