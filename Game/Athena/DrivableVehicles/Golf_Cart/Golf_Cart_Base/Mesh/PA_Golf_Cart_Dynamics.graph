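# Physics asset for the drivable golf cart that defines the jointed bodies for its dangling and swaying accessories.
# A single box on the body bone anchors capsule chains for two hanging dice, a three-segment antenna and a golf bag.
# A three-segment flag chains from the antenna tip, and two clubs attach to the bag.
# Swing limits are loose on flag and dice and tighter on bag and clubs, with mostly minimal twist.
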PhysicsAsset /Game/Athena/DrivableVehicles/Golf_Cart/Golf_Cart_Base/Mesh/PA_Golf_Cart_Dynamics
bodies {
 body [{shapeType=Box, extentX=34.5, extentY=125, extentZ=7.5, center=<-161.1855 0 0>}]
 dyn_antenna_01 [{shapeType=Capsule, length=81.2233, radius=1, center=<0 0 43.4043>}]
 dyn_antenna_02 [{shapeType=Capsule, length=74.2899, radius=0.9588, center=<0 0 36.0111>}]
 dyn_antenna_03 [{shapeType=Capsule, length=32, radius=1, center=<0.0747 -0.0001 15.1849>, rotation=<rot 0.3259 0 0>}]
 dyn_dice_a_01 [{shapeType=Capsule, length=14.7232, radius=0.5, center=<0.0002 0 -8.2213>}]
 dyn_dice_a_03 [{shapeType=Capsule, length=0.5, radius=7.25, center=<0 0 -8.415>}]
 dyn_dice_b_01 [{shapeType=Capsule, length=20.7263, radius=0.5, center=<0.0002 0 -11.994>}]
 dyn_dice_b_03 [{shapeType=Capsule, length=2.25, radius=6.5, center=<0 0 -6>, rotation=<rot 0 0 -70.0003>}]
 dyn_flag_base_01 [{shapeType=Capsule, length=2, radius=2, center=<-1.731 0 0>, rotation=<rot 90 0 0>}]
 dyn_flag_base_02 [{shapeType=Capsule, length=2, radius=2, rotation=<rot 90 0 0>}]
 dyn_flag_base_03 [{shapeType=Capsule, length=2, radius=2, center=<-4.4303 0.0002 0>, rotation=<rot 90 0 0>}]
 dyn_golf_bag_base [{shapeType=Capsule, length=74, radius=15, center=<-2.5 9 1.3>, rotation=<rot 0 -180 -90>}]
 dyn_golf_club_01 [{shapeType=Capsule, length=50, radius=3, center=<0 4.0709 0.4507>, rotation=<rot 0 0 89.0508>}]
 dyn_golf_club_02 [{shapeType=Capsule, length=50, radius=3, center=<0 -1.5 0>, rotation=<rot 0 -14 90>}]
}
constraints [
 {bone1Name=dyn_dice_a_01, bone2Name=body, swing1LimitDegrees=45, swing2LimitDegrees=45, twistLimitDegrees=2}
 {bone1Name=dyn_dice_a_03, bone2Name=dyn_dice_a_01, swing1LimitDegrees=25, swing2LimitDegrees=25, twistLimitDegrees=2}
 {bone1Name=dyn_dice_b_01, bone2Name=body, swing1LimitDegrees=45, swing2LimitDegrees=45, twistLimitDegrees=2}
 {bone1Name=dyn_dice_b_03, bone2Name=dyn_dice_b_01, swing1LimitDegrees=25, swing2LimitDegrees=25, twistLimitDegrees=2}
 {bone1Name=dyn_antenna_01, bone2Name=body, swing1LimitDegrees=20, swing2LimitDegrees=20, twistLimitDegrees=1}
 {bone1Name=dyn_golf_bag_base, bone2Name=body, swing1LimitDegrees=5, swing2LimitDegrees=20, twistLimitDegrees=2}
 {bone1Name=dyn_flag_base_02, bone2Name=dyn_flag_base_01, swing1LimitDegrees=45, swing2LimitDegrees=50, twistLimitDegrees=1}
 {bone1Name=dyn_flag_base_01, bone2Name=dyn_antenna_03, swing1LimitDegrees=45, swing2LimitDegrees=50, twistLimitDegrees=1}
 {bone1Name=dyn_flag_base_03, bone2Name=dyn_flag_base_02, swing1LimitDegrees=45, swing2LimitDegrees=50, twistLimitDegrees=1}
 {bone1Name=dyn_golf_club_01, bone2Name=dyn_golf_bag_base, swing1LimitDegrees=5, swing2LimitDegrees=2, twistLimitDegrees=3}
 {bone1Name=dyn_golf_club_02, bone2Name=dyn_golf_bag_base, swing1LimitDegrees=5, swing2LimitDegrees=2, twistLimitDegrees=30}
 {bone1Name=dyn_antenna_03, bone2Name=dyn_antenna_02, swing1LimitDegrees=20, swing2LimitDegrees=20, twistLimitDegrees=1}
 {bone1Name=dyn_antenna_02, bone2Name=dyn_antenna_01, swing1LimitDegrees=20, swing2LimitDegrees=20, twistLimitDegrees=1}
]
